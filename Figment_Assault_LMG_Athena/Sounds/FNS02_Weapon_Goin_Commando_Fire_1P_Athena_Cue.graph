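# First-person firing loop SoundCue for the Goin Commando LMG in Athena.
# A single looping ballistic fire wave plays through a Modulator that pins pitch to 1, so the loop has no pitch variation.
# The cue uses first-person projectile attenuation, the 1P ranged-fire sound class and first-person gunfire concurrency, with slightly reduced volume.
# The debug flag is left enabled on the cue.
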SoundCue /Figment_Assault_LMG_Athena/Sounds/FNS02_Weapon_Goin_Commando_Fire_1P_Athena_Cue
root=Modulator_0
nodes [
 {id=Modulator_0, children [WavePlayer_1], properties {pitchMax=1, pitchMin=1}}
 {id=WavePlayer_1, properties {bLooping=true, soundWaveAssetPtr=/Figment_Assault_LMG_Athena/Sounds/Wavs/Commando_Ballistic_FireLoop_01}}
]
properties {
 attenuationSettings=&Fort_Weapon_Projectile_FirstPerson_Athena_Att
 bDebug=true
 maxDistance=2097152
 soundClassObject=&Fort_WeaponRanged_Fire_1P_PSM
 volumeMultiplier=0.85
 concurrencySet [&Weapon_Gunfire_FirstPerson_Conc]
}
