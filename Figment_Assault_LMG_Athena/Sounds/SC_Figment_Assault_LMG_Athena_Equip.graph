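# Sound cue for the equip sound of the Figment Assault LMG in Athena.
# A fixed short delay feeds a single wave player that reuses the GoinCommando auto rifle close-range equip sound from the shared weapon sounds.
# The cue plays at about half volume in the weapon foley sound class, using the weapon reload attenuation settings.
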SoundCue /Figment_Assault_LMG_Athena/Sounds/SC_Figment_Assault_LMG_Athena_Equip
root=Delay_0
nodes [
 {id=Delay_0, children [WavePlayer_0], properties {delayMax=0.2, delayMin=0.2}}
 {id=WavePlayer_0, properties {soundWaveAssetPtr=/Game/Sounds/Fort_Weapons/AutoRifle/WPN_GoinCommando_Equip_close_athena}}
]
properties {
 attenuationSettings=&Fort_Weapon_Reload_Att
 maxDistance=2250
 soundClassObject=&Fort_WeaponFoley
 volumeMultiplier=0.501
}
