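# Sound Cue for the looping overheat warning of a mounted turret in the Figment Assault LMG content.
# A looping wave player feeds a continuous modulator that scales pitch and volume from the MiniOverheat parameter.
# Playback uses close weapon projectile attenuation and the low-ammo critical ranged weapon sound class.
# A concurrency override blocks new instances rather than cutting off ones already playing.
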SoundCue /Figment_Assault_LMG_Athena/Sounds/SC_Mounted_Turret_Overheat_Figment
root=ModulatorContinuous_0
nodes [
 {
  id=ModulatorContinuous_0
  children [WavePlayer_0]
  properties {
   pitchModulationParams {default=0, maxOutput=1.1, minOutput=0.85, parameterName=MiniOverheat}
   volumeModulationParams {default=0, parameterName=MiniOverheat}
  }
 }
 {id=WavePlayer_0, properties {bLooping=true, soundWaveAssetPtr=/Figment_Assault_LMG_Athena/Sounds/Wavs/MountedTurret_OverheatWarning_Loop}}
]
properties {
 attenuationSettings=&Fort_Weapon_Projectile_Close_Athena_Att
 bOverrideConcurrency=true
 maxDistance=4950
 soundClassObject=/Game/Sounds/FortSoundClassesAndModes/Fort_WeaponRanged_LowAmmo_Critical_PSM
 concurrencyOverrides {resolutionRule=PreventNew}
}
